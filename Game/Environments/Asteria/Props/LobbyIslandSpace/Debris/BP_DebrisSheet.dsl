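; Configurable debris sheet prop for the Asteria lobby island space environment, set up entirely in the Construction Script, with empty event stubs.
; UseCheapMesh picks which of two static mesh components to add, which is then scaled from the Size parameter.
; A dynamic instance of StarMaterial is applied to that mesh and fed the exposed nebula colors, cloud amount, star texture, brightness and WPO values.
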(event EventBeginPlay)

(event Collision|EventActorBeginOverlap
  (params (OtherActor :type Actor%20Object%20Reference)))

(event EventTick
  (params (DeltaSeconds :type Float%20%28single-precision%29)))

(fn ConstructionScript
  (bind _debrissheetmid (Variables|Default|GetDebrisSheetMID))
  (if (Variables|Parameters|GetUseCheapMesh)
    (bind _returnvalue (AddComponent|NODE_AddStaticMeshComponent-1))
    (bind _output_get (Variables|Default|SetMesh _returnvalue))
    (else
      (bind _returnvalue_2 (AddComponent|NODE_AddStaticMeshComponent-0))
      (bind _output_get_2 (Variables|Default|SetMesh _returnvalue_2))))
  (Transformation|SetRelativeScale3D _output_get (Math|Vector|MakeVector (.x (* (Variables|Parameters|GetSize) 2)) (.z (* (Variables|Parameters|GetSize) 2)) (.z (* (Variables|Parameters|GetSize) 2))))
  (bind _returnvalue_3 (Rendering|Material|CreateDynamicMaterialInstance (Variables|Default|GetMesh) 0 (Variables|Default|GetStarMaterial) "None" :!class /Script/Engine.PrimitiveComponent))
  (Variables|Default|SetDebrisSheetMID _returnvalue_3)
  (Rendering|Material|SetMaterial (Variables|Default|GetMesh) 0 _debrissheetmid)
  (Rendering|Material|SetVectorParameterValue _debrissheetmid "Star_color" (Variables|Parameters|GetNebulaColor) :!class /Script/Engine.MaterialInstanceDynamic)
  (Rendering|Material|SetVectorParameterValue _debrissheetmid "Star_color2" (Variables|Parameters|GetNebulaColor2) :!class /Script/Engine.MaterialInstanceDynamic)
  (Rendering|Material|SetScalarParameterValue _debrissheetmid "Star_cloud" (Variables|Parameters|GetStarSpecialNebula) :!class /Script/Engine.MaterialInstanceDynamic)
  (Rendering|Material|SetTextureParameterValue _debrissheetmid "Star_Tex" (Variables|Parameters|GetStarTexture))
  (Rendering|Material|SetScalarParameterValue _debrissheetmid "star_brightness" (Variables|Parameters|GetBrightness) :!class /Script/Engine.MaterialInstanceDynamic)
  (Rendering|Material|SetScalarParameterValue _debrissheetmid "star_WPO" (Variables|Parameters|GetWPO) :!class /Script/Engine.MaterialInstanceDynamic))
