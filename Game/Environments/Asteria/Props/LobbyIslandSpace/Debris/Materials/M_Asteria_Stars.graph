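# Additive, unlit, two-sided material for stars and galaxy clouds on Asteria lobby island space debris props.
# A panning storm-cloud mask, raised to a power, both distorts the star texture UVs and masks its bright, tinted star channel.
# A second channel forms clouds with a two-color gradient across U, modulated by galaxy noise and time offset by actor position.
# It also computes a normal-direction offset from galaxy noise and a gradient-texture opacity mask.
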Material /Game/Environments/Asteria/Props/LobbyIslandSpace/Debris/Materials/M_Asteria_Stars
graph {
 properties {blendMode=BLEND_Additive, pixelDepthOffsetMode=PDOM_Legacy, shadingModel=MSM_Unlit, twoSided=true, usedShadingModels=Unlit}
 nodes [
  n1 = VectorParameter() {parameterName=Star_color, defaultValue=<rgba 0.9531 0.0086 0.7976 1>}
  n2 = VectorParameter() {parameterName=Star_color2, defaultValue=<rgba 0 0.0776 0.9531 1>}
  n3 = Power(Base=MaterialFunctionCall(Float2=TextureCoordinate()) {materialFunction=&BreakOutFloat2Components}.R)
  n4 = TextureCoordinate()
  n5 = ScalarParameter() {defaultValue=1, parameterName=StarUVtiling}
  n6 = ScalarParameter() {defaultValue=1.5, group=Stars, parameterName=Stars-StarMaskTileB}
  n7 = TextureCoordinate() {coordinateIndex=1}
  n8 = ScalarParameter() {defaultValue=0.3, group=Stars, parameterName=Stars-StarMaskTileBSpeed}
  n9 = TextureSample(UVs=Panner(
    Coordinate=Multiply(A=@n7, B=@n6)
    Time=MaterialFunctionCall(Speed=@n8) {materialFunction=&TimeWithSpeedVariable}.Time
   ) {speedX=0.04, speedY=0.04}) {
   samplerSource=SSM_Wrap_WorldGroupSettings
   samplerType=SAMPLERTYPE_LinearGrayscale
   texture=/Game/Environments/Asteria/Props/LobbyIslandSpace/Debris/Textures/T_StormCloudDetail01_M
  }
  n10 = Power(Base=@n9.R, Exp=ScalarParameter() {defaultValue=3, group=Stars, parameterName=Stars-StarMaskTileBPower})
  n11 = TextureSampleParameter2D(UVs=Add(A=Multiply(A=@n4, B=@n5), B=Multiply(B=@n10) {constA=0.015})) {parameterName=Star_Tex, texture=/Game/Environments/Asteria/Props/LobbyIslandSpace/Debris/Textures/T_Stars_Galaxy_Clouds}
  n12 = TextureSample(UVs=Multiply(A=TextureCoordinate())) {samplerType=SAMPLERTYPE_LinearColor, texture=/Game/Environments/Asteria/Props/LobbyIslandSpace/Debris/Textures/T_Stars_Galaxy_Junior}
  n13 = Sine(None=Multiply(A=ActorPositionWS()) {constB=42.65})
  n14 = Multiply(A=Add(A=Time(), B=Frac(None=@n13)), B=ScalarParameter() {defaultValue=0.075, parameterName=DistortingSpeed}) {constB=0.1}
  n15 = Add(A=@n12.R, B=Sine(None=Frac(None=@n14)))
  n16 = Add(A=ScalarParameter() {defaultValue=0.2, parameterName=Star_cloud}, B=Multiply(A=Saturate(None=@n15)) {constB=0.1})
  n17 = VectorParameter() {parameterName=Star_color, defaultValue=<rgba 0.9531 0.1625 0.8835 1>}
  n18 = Constant3Vector() {constant=<rgba 0.1496 0.4274 0.75 0>}
  n19 = Multiply(
   A=LinearInterpolate(A=LinearInterpolate(A=@n17.RGB, B=@n18, Alpha=@n12.R), B=Constant3Vector() {constant=<rgba 0.147 0.4179 0.7454 1>}, Alpha=@n12.B)
   B=Multiply(A=@n11.R, B=@n10)
  )
  n20 = ScalarParameter() {defaultValue=2, parameterName=star_contrast}
  Add(
   A=Multiply(A=LinearInterpolate(A=@n1.RGB, B=@n2.RGB, Alpha=@n3), B=Multiply(A=@n11.G, B=@n16))
   B=Multiply(A=Power(Base=@n19, Exp=@n20), B=ScalarParameter() {defaultValue=50, parameterName=star_Brightness})
  )
  Frac(None=@n15)
  Multiply(A=Multiply(A=@n12.G, B=ScalarParameter() {defaultValue=1500, parameterName=Star_WPO}), B=VertexNormalWS())
  n21 = TextureSampleParameter2D() {
   parameterName=GradientType
   samplerType=SAMPLERTYPE_LinearGrayscale
   texture=/Game/Environments/Asteria/Props/LobbyIslandSpace/Debris/Textures/T_SpaceDecalMask
  }
  Multiply(A=ScalarParameter() {defaultValue=0.9032, group=Global, parameterName=star_Opacity}, B=@n21.R)
 ]
}
